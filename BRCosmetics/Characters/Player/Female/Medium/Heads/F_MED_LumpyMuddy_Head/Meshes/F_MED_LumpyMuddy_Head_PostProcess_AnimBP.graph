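# Post-process AnimBP for the LumpyMuddy female head cosmetic that counter-rotates the head and adds spring physics to the dyn_Coffee bone.
# Each update reads the pawn mesh's head delta from ref pose against neck_02, negating and scaling it per axis.
# The AnimGraph adds that rotation to the head, then runs damped, angle-limited Anim Dynamics on dyn_Coffee.
# The update also sets a front-end flag, FE, which nothing shown reads.
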AnimBlueprint /BRCosmetics/Characters/Player/Female/Medium/Heads/F_MED_LumpyMuddy_Head/Meshes/F_MED_LumpyMuddy_Head_PostProcess_AnimBP
target_skeleton=&/Fortnite_Head_LumpyMuddy_Skeleton
graphs [
 AnimGraph {nodes [
   n1 = Input Pose()
   n2 = Anim Dynamics - Bone: dyn_Coffee(
    ComponentPose="Transform (Modify) Bone - Bone: head"(ComponentPose=Local To Component(LocalPose=@n1), Alpha=1, Translation="0,0,0", Rotation=Get head rot() -> (head rot:Rotator), Scale="1,1,1") {rotationMode=BMM_Additive, rotationSpace=BCS_BoneSpace, boneToModify {boneName=head}}
    Alpha=1
    ExternalForce="0,0,0"
   ) {
    angularDampingOverride=0.999
    angularSpringConstant=20000
    bAngularSpring=true
    bOverrideAngularDamping=true
    gravityScale=0.2
    simulationSpace=BoneRelative
    boundBone {boneName=dyn_Coffee}
    componentAppliedLinearAccClamp=<500 500 500>
    componentLinearAccScale=<0.2 0.2 0.2>
    componentLinearVelScale=<0.2 0.2 0.2>
    gravityOverride=<500 500 500>
    physicsBodyDefinitions [{
      boundBone {boneName=dyn_Coffee}
      boxExtents=<40 40 30>
      constraintSetup {angularTargetAxis=AxisZ, twistAxis=AxisZ, angularLimitsMax=<5 5 20>, angularLimitsMin=<-5 -5 -20>, angularTarget=<0 0 1>}
      localJointOffset=<0 0 0>
     }]
    relativeSpaceBone {boneName=head}
    simSpaceSettings {maxAngularAcceleration=10, maxAngularVelocity=2, simSpaceAngularAlpha=0.5, externalAngularVelocity=<5 5 5>}
   }
   Output Pose(Result=Component To Local(ComponentPose=@n2))
  ]}
 EventGraph {nodes [
   Event BlueprintUpdateAnimation()
   > n1 = Sequence()
   n2 = Get bIsInFrontEnd()
   > Set FE(FE:double="To Float (Boolean)"(InBool=NOT Boolean(A=@n2))) -> (Output_Get:double)
   n3 = Cast To FortPlayerPawn(execute=@n1.then_1, Object=TryGetPawnOwner())
   n4 = Get Mesh(self=@n3)
   n5 = BreakRotator(InRot=BreakTransform(InTransform=GetDeltaTransformFromRefPose(self=@n4, BoneName=head, BaseName=neck_02)).Rotation)
   n6 = Get Y percentage() -> (Y percentage:float)
   n7 = Get X percentage() -> (X percentage:float)
   n8 = Get Z percentage() -> (Z percentage:float)
   > Set head rot(head rot:Rotator=MakeRotator(
     Roll=float * float!(A=@n5, B=-1, C:double=@n7) -> (ReturnValue)
     Pitch=float * float(A=@n5.Pitch, B=-1, C=@n6)
     Yaw=float * float(A=@n5.Yaw, B=-1, C=@n8)
    )) -> (Output_Get:Rotator)
   TryGetPawnOwner()
  ]}
]
variables [
 FE {type=number}
 head rot {type=Rotator}
 Head Rotate 0 pos {type=Rotator}
 Neck Rotate {type=number}
 Neck Rotate_0_pos {type=Rotator}
 Neck Rotate_1_neg {type=Rotator}
 Spine Rotate 0 Pos {type=Rotator}
 X percentage {type=number}
 Y percentage {type=number, default=1}
 Z percentage {type=number, default=0.6}
]
